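# Sound cue for the Vuvuzela pickaxe's goal sound in Athena.
# A quality-level node at the root plays the standard goal wave for the first two tiers and the "opt" variant for the last two.
# It uses a soccer attenuation asset named as non-spatialized and is routed to the Athena indoor ambient sound class rather than a weapon class.
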SoundCue /Game/Athena/Sounds/Weapons/PickAxes/Vuvuzela/Vuvuzela_Goal_Cue
root=QualityLevel_0
nodes [
 {id=QualityLevel_0, children [WavePlayer_0,WavePlayer_0,WavePlayer_1,WavePlayer_1]}
 {id=WavePlayer_0, properties {soundWaveAssetPtr=&/Vuvuzela_Goal_01}}
 {id=WavePlayer_1, properties {soundWaveAssetPtr=&/Vuvuzela_Goal_01_opt}}
]
properties {
 attenuationSettings=/Game/Sounds/Attenuation/Fort_Soccer_NoSpatial_Athena_Att
 maxDistance=6383.4253
 soundClassObject=&Fort_Ambient_Indoors_Athena
 volumeMultiplier=1.002
}
